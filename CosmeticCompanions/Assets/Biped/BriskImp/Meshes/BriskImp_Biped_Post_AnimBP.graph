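# Post-process AnimBP for the BriskImp biped companion that adds corrective poses and secondary motion on top of the incoming pose.
# A Pose Blend node applies the BriskImp pose asset, then two Constraint nodes drive the left and right back spikes from weighted spine, jaw, neck and shoulder bones.
# Chained per-bone AnimDynamics nodes add spring-damped, root-relative jiggle to ears and tail, with looser limits toward the tips and alpha set by inverted DynamicsAlpha.
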AnimBlueprint /CosmeticCompanions/Assets/Biped/BriskImp/Meshes/BriskImp_Biped_Post_AnimBP
target_skeleton=&/BriskImp_Biped_Skeleton
graphs [
 AnimGraph {nodes [
   Pose!(PoseAsset:Pose Asset Object Reference=Get PoseAsset() -> (PoseAsset:Pose Asset Object Reference), SourcePose:Pose Link Structure) -> (Pose:Pose Link Structure) {is=title AnimGraphNode_PoseBlendNode, bIgnorePoseAssetSourceAnimationCurves=true, blendOption=Linear}
   n1 = Input Pose()
   n2 = Get PoseAsset()
   n3 = "Constraint - backSpike_L (PPPPPPPPPPPP)"!(
    ComponentPose=Local To Component(LocalPose=Pose(PoseAsset=@n2, SourcePose=@n1) {bIgnorePoseAssetSourceAnimationCurves=true, blendOption=Linear})
    Alpha=1
    ConstraintWeights_0=0.4455
    ConstraintWeights_1:float=0.1686
    ConstraintWeights_2:float=0.1469
    ConstraintWeights_3:float=0.084
    ConstraintWeights_4:float=0.0572
    ConstraintWeights_5:float=0.0374
    ConstraintWeights_6:float=0.0321
    ConstraintWeights_7:float=0.0181
    ConstraintWeights_8:float=0.0074
    ConstraintWeights_9:float=0.0016
    ConstraintWeights_10:float=0.0008
    ConstraintWeights_11:float=0.0003
   ) -> (Pose) {
    boneToModify {boneName=backSpike_L}
    constraintSetup [
     {targetBone {boneName=spine_05}}
     {targetBone {boneName=clavicle_l}}
     {targetBone {boneName=spine_04}}
     {targetBone {boneName=reverse_jaw}}
     {targetBone {boneName=spine_03}}
     {targetBone {boneName=neck_01}}
     {targetBone {boneName=upperarm_twist_01_l}}
     {targetBone {boneName=clavicle_r}}
     {targetBone {boneName=spine_02}}
     {targetBone {boneName=spine_01}}
     {targetBone {boneName=upperarm_l}}
     {targetBone {boneName=upperarm_twist_02_l}}
    ]
    constraintWeights [1,1,1,1,1,1,1,1,1,1,1,1]
   }
   n4 = "Constraint - backSpike_R (PPPPPPPPPPPP)"!(
    ComponentPose=@n3
    Alpha=1
    ConstraintWeights_0=0.4455
    ConstraintWeights_1:float=0.1686
    ConstraintWeights_2:float=0.1469
    ConstraintWeights_3:float=0.084
    ConstraintWeights_4:float=0.0572
    ConstraintWeights_5:float=0.0374
    ConstraintWeights_6:float=0.0321
    ConstraintWeights_7:float=0.0181
    ConstraintWeights_8:float=0.0074
    ConstraintWeights_9:float=0.0016
    ConstraintWeights_10:float=0.0008
    ConstraintWeights_11:float=0.0003
   ) -> (Pose) {
    boneToModify {boneName=backSpike_R}
    constraintSetup [
     {targetBone {boneName=spine_05}}
     {targetBone {boneName=clavicle_r}}
     {targetBone {boneName=spine_04}}
     {targetBone {boneName=reverse_jaw}}
     {targetBone {boneName=spine_03}}
     {targetBone {boneName=neck_01}}
     {targetBone {boneName=upperarm_twist_01_r}}
     {targetBone {boneName=clavicle_l}}
     {targetBone {boneName=spine_02}}
     {targetBone {boneName=spine_01}}
     {targetBone {boneName=upperarm_r}}
     {targetBone {boneName=upperarm_twist_02_r}}
    ]
    constraintWeights [1,1,1,1,1,1,1,1,1,1,1,1]
   }
   n5 = Get DynamicsAlpha() -> (DynamicsAlpha:float)
   n6 = Anim Dynamics - Bone: ear_l_01(ComponentPose=@n4, Alpha=@n5, ExternalForce="0,0,0") {
    angularDampingOverride=1
    angularSpringConstant=20000
    bAngularSpring=true
    bOverrideAngularDamping=true
    bUseGravityOverride=true
    gravityScale=0.35
    simulationSpace=RootRelative
    alphaScaleBiasClamp {bias=1, scale=-1}
    boundBone {boneName=ear_l_01}
    componentLinearAccScale=<1 1 1>
    componentLinearVelScale=<1 1 1>
    gravityOverride {z=-980}
    physicsBodyDefinitions [{
      boundBone {boneName=ear_l_01}
      boxExtents=<20 30 30>
      constraintSetup {angularLimitsMax=<15 15 15>, angularLimitsMin=<-15 -15 -15>, angularTarget=<1 0 0>}
      localJointOffset=<4 0 0>
     }]
   }
   n7 = Anim Dynamics - Bone: ear_l_02(ComponentPose=@n6, Alpha=@n5, ExternalForce="0,0,0") {
    angularDampingOverride=1
    angularSpringConstant=20000
    bAngularSpring=true
    bOverrideAngularDamping=true
    bUseGravityOverride=true
    gravityScale=0.35
    simulationSpace=RootRelative
    alphaScaleBiasClamp {bias=1, scale=-1}
    boundBone {boneName=ear_l_02}
    componentLinearAccScale=<1 1 1>
    componentLinearVelScale=<1 1 1>
    gravityOverride {z=-980}
    physicsBodyDefinitions [{
      boundBone {boneName=ear_l_02}
      boxExtents=<20 30 30>
      constraintSetup {angularLimitsMax=<25 25 25>, angularLimitsMin=<-25 -25 -25>, angularTarget=<1 0 0>}
      localJointOffset=<4 0 0>
     }]
   }
   n8 = Anim Dynamics - Bone: ear_r_01(ComponentPose=@n7, Alpha=@n5, ExternalForce="0,0,0") {
    angularDampingOverride=1
    angularSpringConstant=20000
    bAngularSpring=true
    bOverrideAngularDamping=true
    bUseGravityOverride=true
    gravityScale=0.35
    simulationSpace=RootRelative
    alphaScaleBiasClamp {bias=1, scale=-1}
    boundBone {boneName=ear_r_01}
    componentLinearAccScale=<1 1 1>
    componentLinearVelScale=<1 1 1>
    gravityOverride {z=-980}
    physicsBodyDefinitions [{
      boundBone {boneName=ear_r_01}
      boxExtents=<20 30 30>
      constraintSetup {angularLimitsMax=<15 15 15>, angularLimitsMin=<-15 -15 -15>, angularTarget=<1 0 0>}
      localJointOffset=<4 0 0>
     }]
   }
   n9 = Anim Dynamics - Bone: ear_r_02(ComponentPose=@n8, Alpha=@n5, ExternalForce="0,0,0") {
    angularDampingOverride=1
    angularSpringConstant=20000
    bAngularSpring=true
    bOverrideAngularDamping=true
    bUseGravityOverride=true
    gravityScale=0.35
    simulationSpace=RootRelative
    alphaScaleBiasClamp {bias=1, scale=-1}
    boundBone {boneName=ear_r_02}
    componentLinearAccScale=<1 1 1>
    componentLinearVelScale=<1 1 1>
    gravityOverride {z=-980}
    physicsBodyDefinitions [{
      boundBone {boneName=ear_r_02}
      boxExtents=<20 30 30>
      constraintSetup {angularLimitsMax=<25 25 25>, angularLimitsMin=<-25 -25 -25>, angularTarget=<1 0 0>}
      localJointOffset=<4 0 0>
     }]
   }
   n10 = Anim Dynamics - Bone: tail_01(ComponentPose=@n9, Alpha=@n5, ExternalForce="0,0,0") {
    angularDampingOverride=1
    angularSpringConstant=20000
    bAngularSpring=true
    bOverrideAngularDamping=true
    bUseGravityOverride=true
    gravityScale=0.35
    simulationSpace=RootRelative
    alphaScaleBiasClamp {bias=1, scale=-1}
    boundBone {boneName=tail_01}
    componentLinearAccScale=<1 1 1>
    componentLinearVelScale=<1 1 1>
    gravityOverride {z=-980}
    physicsBodyDefinitions [{
      boundBone {boneName=tail_01}
      boxExtents=<20 30 30>
      constraintSetup {angularLimitsMax=<20 20 20>, angularLimitsMin=<-20 -20 -20>, angularTarget=<1 0 0>}
      localJointOffset=<4 0 0>
     }]
   }
   n11 = Anim Dynamics - Bone: tail_02(ComponentPose=@n10, Alpha=@n5, ExternalForce="0,0,0") {
    angularDampingOverride=1
    angularSpringConstant=20000
    bAngularSpring=true
    bOverrideAngularDamping=true
    bUseGravityOverride=true
    gravityScale=0.35
    simulationSpace=RootRelative
    alphaScaleBiasClamp {bias=1, scale=-1}
    boundBone {boneName=tail_02}
    componentLinearAccScale=<1 1 1>
    componentLinearVelScale=<1 1 1>
    gravityOverride {z=-980}
    physicsBodyDefinitions [{
      boundBone {boneName=tail_02}
      boxExtents=<20 30 30>
      constraintSetup {angularLimitsMax=<35 35 35>, angularLimitsMin=<-35 -35 -35>, angularTarget=<1 0 0>}
      localJointOffset=<4 0 0>
     }]
   }
   n12 = Anim Dynamics - Bone: tail_03(ComponentPose=@n11, Alpha=@n5, ExternalForce="0,0,0") {
    angularDampingOverride=1
    angularSpringConstant=20000
    bAngularSpring=true
    bOverrideAngularDamping=true
    bUseGravityOverride=true
    gravityScale=0.35
    simulationSpace=RootRelative
    alphaScaleBiasClamp {bias=1, scale=-1}
    boundBone {boneName=tail_03}
    componentLinearAccScale=<1 1 1>
    componentLinearVelScale=<1 1 1>
    gravityOverride {z=-980}
    physicsBodyDefinitions [{
      boundBone {boneName=tail_03}
      boxExtents=<20 30 30>
      constraintSetup {angularLimitsMax=<45 45 45>, angularLimitsMin=<-45 -45 -45>, angularTarget=<1 0 0>}
      localJointOffset=<4 0 0>
     }]
   }
   Output Pose(Result=Component To Local(ComponentPose=@n12))
  ]}
 EventGraph {links [], nodes []}
]
variables [PoseAsset {type=PoseAsset, default=&/BriskImp_Biped_PoseAsset}]
